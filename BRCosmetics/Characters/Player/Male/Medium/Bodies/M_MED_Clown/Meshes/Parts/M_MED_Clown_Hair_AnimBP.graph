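# Animation Blueprint for the clown body's hair part, adding secondary physics motion to the hair bones.
# It copies the pose from the attached parent mesh, then runs it through a chain of single-bone Anim Dynamics nodes.
# The A, B and C strand bones use stiff angular springs with reduced gravity, while the two hair-part bones have no spring.
# All are heavily damped with angular limits; the event graph has no logic.
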AnimBlueprint /BRCosmetics/Characters/Player/Male/Medium/Bodies/M_MED_Clown/Meshes/Parts/M_MED_Clown_Hair_AnimBP
target_skeleton=&/M_MED_Clown_Hair_Skeleton
graphs [
 AnimGraph {nodes [
   n1 = Anim Dynamics - Bone: dyn_hair_A_01(ComponentPose=Local To Component(LocalPose=Copy Pose From Mesh() {bUseAttachedParent=true}), Alpha=1, ExternalForce="0,0,0") {
    angularDampingOverride=0.99
    angularSpringConstant=2500
    bAngularSpring=true
    bGravityOverrideInSimSpace=true
    bOverrideAngularDamping=true
    gravityScale=0.05
    alphaBoolBlend {blendOption=Linear}
    boundBone {boneName=dyn_hair_A_01}
    componentLinearAccScale=<0.2 0.2 0.2>
    componentLinearVelScale=<0.2 0.2 0.2>
    physicsBodyDefinitions [{
      sphereCollisionRadius=0
      boundBone {boneName=dyn_hair_A_01}
      boxExtents=<10 20 20>
      constraintSetup {angularLimitsMax=<0 20 20>, angularLimitsMin=<0 0 -20>, angularTarget=<0 0 0>}
      localJointOffset=<2 0 1>
     }]
   }
   n2 = Anim Dynamics - Bone: dyn_hair_A_02(ComponentPose=@n1, Alpha=1, ExternalForce="0,0,0") {
    angularDampingOverride=0.99
    angularSpringConstant=2500
    bAngularSpring=true
    bGravityOverrideInSimSpace=true
    bOverrideAngularDamping=true
    gravityScale=0.05
    alphaBoolBlend {blendOption=Linear}
    boundBone {boneName=dyn_hair_A_02}
    componentLinearAccScale=<0.2 0.2 0.2>
    componentLinearVelScale=<0.2 0.2 0.2>
    physicsBodyDefinitions [{
      sphereCollisionRadius=0
      boundBone {boneName=dyn_hair_A_02}
      boxExtents=<10 20 20>
      constraintSetup {angularLimitsMax=<0 20 20>, angularLimitsMin=<0 0 -20>, angularTarget=<0 0 0>}
      localJointOffset=<2 0 1>
     }]
   }
   n3 = Anim Dynamics - Bone: dyn_hair_B_01(ComponentPose=@n2, Alpha=1, ExternalForce="0,0,0") {
    angularDampingOverride=0.99
    angularSpringConstant=2500
    bAngularSpring=true
    bGravityOverrideInSimSpace=true
    bOverrideAngularDamping=true
    gravityScale=0.05
    alphaBoolBlend {blendOption=Linear}
    boundBone {boneName=dyn_hair_B_01}
    componentLinearAccScale=<0.2 0.2 0.2>
    componentLinearVelScale=<0.2 0.2 0.2>
    physicsBodyDefinitions [{
      sphereCollisionRadius=0
      boundBone {boneName=dyn_hair_B_01}
      boxExtents=<10 20 20>
      constraintSetup {angularLimitsMax=<0 20 20>, angularLimitsMin=<0 0 -20>, angularTarget=<0 0 0>}
      localJointOffset=<2 0 1>
     }]
   }
   n4 = Anim Dynamics - Bone: dyn_hair_B_02(ComponentPose=@n3, Alpha=1, ExternalForce="0,0,0") {
    angularDampingOverride=0.99
    angularSpringConstant=2500
    bAngularSpring=true
    bGravityOverrideInSimSpace=true
    bOverrideAngularDamping=true
    gravityScale=0.05
    alphaBoolBlend {blendOption=Linear}
    boundBone {boneName=dyn_hair_B_02}
    componentLinearAccScale=<0.2 0.2 0.2>
    componentLinearVelScale=<0.2 0.2 0.2>
    physicsBodyDefinitions [{
      sphereCollisionRadius=0
      boundBone {boneName=dyn_hair_B_02}
      boxExtents=<10 20 20>
      constraintSetup {angularLimitsMax=<0 20 20>, angularLimitsMin=<0 0 -20>, angularTarget=<0 0 0>}
      localJointOffset=<2 0 1>
     }]
   }
   n5 = Anim Dynamics - Bone: dyn_hair_C_01(ComponentPose=@n4, Alpha=1, ExternalForce="0,0,0") {
    angularDampingOverride=0.99
    angularSpringConstant=2500
    bAngularSpring=true
    bGravityOverrideInSimSpace=true
    bOverrideAngularDamping=true
    gravityScale=0.05
    alphaBoolBlend {blendOption=Linear}
    boundBone {boneName=dyn_hair_C_01}
    componentLinearAccScale=<0.2 0.2 0.2>
    componentLinearVelScale=<0.2 0.2 0.2>
    physicsBodyDefinitions [{
      sphereCollisionRadius=0
      boundBone {boneName=dyn_hair_C_01}
      boxExtents=<10 20 20>
      constraintSetup {angularLimitsMax=<0 20 20>, angularLimitsMin=<0 0 -20>, angularTarget=<0 0 0>}
      localJointOffset=<2 0 1>
     }]
   }
   n6 = Anim Dynamics - Bone: dyn_hair_C_02(ComponentPose=@n5, Alpha=1, ExternalForce="0,0,0") {
    angularDampingOverride=0.99
    angularSpringConstant=2500
    bAngularSpring=true
    bGravityOverrideInSimSpace=true
    bOverrideAngularDamping=true
    gravityScale=0.05
    alphaBoolBlend {blendOption=Linear}
    boundBone {boneName=dyn_hair_C_02}
    componentLinearAccScale=<0.2 0.2 0.2>
    componentLinearVelScale=<0.2 0.2 0.2>
    physicsBodyDefinitions [{
      sphereCollisionRadius=0
      boundBone {boneName=dyn_hair_C_02}
      boxExtents=<10 20 20>
      constraintSetup {angularLimitsMax=<0 20 20>, angularLimitsMin=<0 0 -20>, angularTarget=<0 0 0>}
      localJointOffset=<2 0 1>
     }]
   }
   n7 = Anim Dynamics - Bone: dyn_hair_part_A(ComponentPose=@n6, Alpha=1, ExternalForce="0,0,0") {
    angularDampingOverride=0.99
    bGravityOverrideInSimSpace=true
    bOverrideAngularDamping=true
    alphaBoolBlend {blendOption=Linear}
    boundBone {boneName=dyn_hair_part_A}
    componentLinearAccScale=<0.2 0.2 0.2>
    componentLinearVelScale=<0.2 0.2 0.2>
    physicsBodyDefinitions [{
      sphereCollisionRadius=0
      boundBone {boneName=dyn_hair_part_A}
      boxExtents=<10 10 1>
      constraintSetup {angularLimitsMax=<0 20 20>, angularLimitsMin=<0 0 -20>, angularTarget=<0 0 0>}
      localJointOffset=<0 0 -2>
     }]
   }
   n8 = Anim Dynamics - Bone: dyn_hair_part_B(ComponentPose=@n7, Alpha=1, ExternalForce="0,0,0") {
    angularDampingOverride=0.99
    bGravityOverrideInSimSpace=true
    bOverrideAngularDamping=true
    alphaBoolBlend {blendOption=Linear}
    boundBone {boneName=dyn_hair_part_B}
    componentLinearAccScale=<0.2 0.2 0.2>
    componentLinearVelScale=<0.2 0.2 0.2>
    physicsBodyDefinitions [{
      sphereCollisionRadius=0
      boundBone {boneName=dyn_hair_part_B}
      boxExtents=<10 10 1>
      constraintSetup {angularLimitsMax=<0 20 20>, angularLimitsMin=<0 0 -20>, angularTarget=<0 0 0>}
      localJointOffset=<0 0 -2>
     }]
   }
   Output Pose(Result=Component To Local(ComponentPose=@n8)) {layerGroup=None}
  ]}
 EventGraph {nodes [Event BlueprintUpdateAnimation()]}
]
variables []
